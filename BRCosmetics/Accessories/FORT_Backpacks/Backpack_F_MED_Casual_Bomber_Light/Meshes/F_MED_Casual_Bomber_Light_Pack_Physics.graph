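# Physics asset for the Casual Bomber Light backpack cosmetic, defining capsule bodies for the pack and its dangling wing, tail and mouth joints.
# Two-joint left and right wing chains and a three-joint tail chain hang off the pack body with tight, mirrored swing and twist limits.
# The mouth attaches to the pack as a hinge, with swing1 and twist locked and only swing2 left to move.
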PhysicsAsset /BRCosmetics/Accessories/FORT_Backpacks/Backpack_F_MED_Casual_Bomber_Light/Meshes/F_MED_Casual_Bomber_Light_Pack_Physics
bodies {
 dyn_L_wing_jnt_1 [{shapeType=Capsule, length=2, radius=1}]
 dyn_L_wing_jnt_2 [{shapeType=Capsule, length=4, radius=2, rotation=<rot 90 -180 -180>}]
 dyn_R_wing_jnt_1 [{shapeType=Capsule, length=2, radius=1}]
 dyn_R_wing_jnt_2 [{shapeType=Capsule, length=4, radius=2, rotation=<rot 90 180 180>}]
 dyn_mouth_jnt [{shapeType=Capsule, length=1, radius=1.7, center=<0.5 0 0>, rotation=<rot 0 0 90>}]
 dyn_tail_jnt_1 [{shapeType=Capsule, length=1, radius=1, center=<0.743 0 0.1454>, rotation=<rot 78.9276 180 180>}]
 dyn_tail_jnt_2 [{shapeType=Capsule, length=1, radius=1, center=<0.743 0 0.1454>, rotation=<rot 78.9274 180 180>}]
 dyn_tail_jnt_3 [{shapeType=Capsule, length=1, radius=1, center=<0.743 0 0.1454>, rotation=<rot 78.9276 180 180>}]
 pack [{shapeType=Capsule, length=1, radius=1}]
}
constraints [
 {bone1Name=dyn_tail_jnt_1, bone2Name=pack, swing1LimitDegrees=5, swing2LimitDegrees=10, twistLimitDegrees=4}
 {bone1Name=dyn_tail_jnt_2, bone2Name=dyn_tail_jnt_1, swing1LimitDegrees=5, swing2LimitDegrees=10, twistLimitDegrees=4}
 {bone1Name=dyn_tail_jnt_3, bone2Name=dyn_tail_jnt_2, swing1LimitDegrees=5, swing2LimitDegrees=10, twistLimitDegrees=4}
 {
  bone1Name=dyn_mouth_jnt
  bone2Name=pack
  swing1LimitDegrees=45
  swing1Motion=Locked
  swing2LimitDegrees=50
  twistLimitDegrees=45
  twistMotion=Locked
 }
 {bone1Name=dyn_R_wing_jnt_1, bone2Name=pack, swing1LimitDegrees=25, swing2LimitDegrees=10, twistLimitDegrees=10}
 {bone1Name=dyn_R_wing_jnt_2, bone2Name=dyn_R_wing_jnt_1, swing1LimitDegrees=20, swing2LimitDegrees=5, twistLimitDegrees=10}
 {bone1Name=dyn_L_wing_jnt_1, bone2Name=pack, swing1LimitDegrees=25, swing2LimitDegrees=10, twistLimitDegrees=10}
 {bone1Name=dyn_L_wing_jnt_2, bone2Name=dyn_L_wing_jnt_1, swing1LimitDegrees=20, swing2LimitDegrees=5, twistLimitDegrees=10}
]
